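# Niagara system for the IronBlaze glider's thruster effects, attached to skeletal mesh sockets.
# Short-lived flame meshes, cylinder-sampled air sprites and flickering glow sprites with lights and coronas follow sockets by partition index.
# Anim-trail ribbons draw side-wing streaks, with velocity flipped in the front end.
# A system timeline and size scale drive fade-in and intensity, with time-of-day and front-end modifiers adjusting color and brightness.
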NiagaraSystem /BRCosmetics/Gadgets/Assets/VinderTech_GliderChute/Glider_IronBlaze/FX/NS_Glider_IronBlaze_Thrusters
dependencies {
 usedDataInterfaces [
  &NiagaraDataInterfaceCurve
  &NiagaraDataInterfaceColorCurve
  &NiagaraDataInterfaceArrayFloat
  &NiagaraDataInterfaceArrayFloat3
 ]
 usedDynamicInputs [&Boolean_AndOperation, &SetBoolByFloatComparison]
 usedModules [
  &SystemState
  &Timeline
  &self:SystemSpawnScript.SystemScriptSource.SystemScriptGraph.NiagaraNodeAssignment_0.SetVariables_#1
 ]
}
emitters {
 AirThrusters003 {
  emitterUpdateScript [
   &EmitterState()
   &SpawnBurst_Instantaneous(Spawn Count=12, Use Spawn Probability=false, UseLoopCountLimit=false) {enabled=false}
   &SpawnRate(SpawnRate=300, Use Spawn Probability=false)
   &SpawnBurst_Instantaneous(Spawn Count=12, Spawn Time=3.1884, Use Spawn Probability=false, UseLoopCountLimit=false) {enabled=false, module=SpawnBurst_Instantaneous001}
   &SpawnBurst_Instantaneous(Spawn Count=12, Spawn Time=1.01, Use Spawn Probability=false, UseLoopCountLimit=false) {enabled=false, module=SpawnBurst_Instantaneous002}
  ]
  particleSpawnScript [
   &PartitionParticles(Partition Mode:ENiagara_PartitionMode=::NewEnumerator1)
   &InitializeParticle(
    Color=<rgba 0.6588 2 1.2078 0.1>
    Lifetime Max=0.75
    Lifetime Min=0.25
    Lifetime Mode=::NewEnumerator1
    Sprite Size Mode=::NewEnumerator3
    Uniform Sprite Size=4
    UsePositionOffset=false
    Write Material Random=true
   )
   /SetVariables(Particles.particleIndex:NiagaraFloat=&SelectFloatFromArray(
     Array Sampling Mode=::NewEnumerator0
     Direct Array Index=&ReturnParticleIDAttributes()
     Direct Array Mode=::NewEnumerator1
     Float Selection Array=DataInterface() {propertyValues {FloatData [0,0.15,0.3,0.45,0.6,0.75,0.9,1]}}
     Random Seed=0
    )) {enabled=false, isSetParameters=true}
   &SpriteFacingAndAlignment(
    Sprite Facing=&SelectVectorFromArray(
     Array Sampling Mode=::NewEnumerator0
     Direct Array Index=$Particles.PartitionParticles.Partition
     Vector Selection Array=DataInterface() {propertyValues {FloatData [<0 0.5 0>, <0 0.5 0>]}}
     Random Seed=0
    )
    WriteAlignment=false
    WriteFacing=true
   )
  ]
  particleUpdateScript [
   &ParticleState()
   &SkeletalMeshLocation(
    Bone / Socket Index=$Particles.PartitionParticles.Partition
    MaskSpawnGroup=false
    Mesh Sampling Type=::NewEnumerator5
    Skeletal Mesh=$User.SkeletalMesh
    Socket Sampling Mode=::NewEnumerator5
   )
   &ShapeLocation(
    Cylinder Height=3
    Cylinder Radius=12
    MaskSpawnGroup=false
    Offset=&SelectVectorFromArray(
     Direct Array Index=$Particles.PartitionParticles.Partition
     Vector Selection Array=DataInterface() {propertyValues {FloatData [<0 0 -1>, <0 0 -1>]}}
    )
    Offset Mode=::NewEnumerator0
    Override Seed=false
    Rotation Axis=&SelectVectorFromArray(
     Direct Array Index=$Particles.PartitionParticles.Partition
     Vector Selection Array=DataInterface() {propertyValues {FloatData [<0 0.5 0>, <0 0.5 0>]}}
    )
    Rotation Mode=::NewEnumerator1
    Shape Origin=$Output.SkeletalMeshLocation.SampledPositionWithoutOffset
    Shape Primitive=::NewEnumerator1
    Surface Only=true
    Use Endcaps In Surface Only Mode=false
    UseOwnerScale=true
   )
   &ScaleColor(
    Scale Alpha=&Multiply_Float(
     A=&FloatFromCurve(FloatCurve=DataInterface() {propertyValues {Curve {keys [
          {arriveTangent=6.8336, interpMode=RCIM_Cubic, leaveTangent=6.8336, tangentMode=RCTM_User, time=0, value=0}
          <at 0.2 1 RCIM_Cubic>
          {arriveTangent=-1.25, interpMode=RCIM_Cubic, leaveTangent=-1.25, time=0.75, value=0.8}
          {arriveTangent=-4.31, interpMode=RCIM_Cubic, leaveTangent=-4.31, tangentMode=RCTM_User, time=1, value=0}
         ]}}})
     B=&Lerp_Float(A=0, B=1, Alpha=$System.Timeline.NormalizedTime)
    )
    Scale RGB=&VectorFromFloat(Value=&Multiply_Float(
      A=&FloatFromCurve(
       FloatCurve=DataInterface() {propertyValues {Curve {keys [
           {arriveTangent=6.8336, interpMode=RCIM_Cubic, leaveTangent=6.8336, tangentMode=RCTM_User, time=0, value=0}
           <at 0.2 1 RCIM_Cubic>
           {arriveTangent=-1.25, interpMode=RCIM_Cubic, leaveTangent=-1.25, time=0.75, value=0.8}
           {arriveTangent=-4.31, interpMode=RCIM_Cubic, leaveTangent=-4.31, tangentMode=RCTM_User, time=1, value=0}
          ]}}}
       Scale Curve=10
      )
      B=&Lerp_Float(A=0, B=1, Alpha=$System.Timeline.NormalizedTime)
     ))
    ScaleA=true
    ScaleRGB=true
   )
   &DynamicMaterialParameters(Write Parameter Index 0=false) {enabled=false}
  ]
  rendererClasses [&NiagaraSpriteRendererProperties]
  renderers [{rendererClass=&NiagaraSpriteRendererProperties}]
 }
 FakeLighting {
  emitterUpdateScript [
   &EmitterState()
   &SpawnBurst_Instantaneous(Spawn Count=4, Use Spawn Probability=false, UseLoopCountLimit=true)
  ]
  particleSpawnScript [
   &PartitionParticles()
   &InitializeParticle(Color=<rgba 1 1 1 0.5>, Sprite Size Mode=::NewEnumerator3, Uniform Sprite Size=55, UsePositionOffset=false, Write Material Random=true)
   &DynamicMaterialParameters(
    Index 0 Param 1=1
    Index 0 Param 2=1
    Index 0 Param 3=1
    Index 0 Param 4=1
    Module.Param0WriteEnabled=true
    Module.Param1WriteEnabled=true
    Module.Param2WriteEnabled=true
    Module.Param3WriteEnabled=true
   )
  ]
  particleUpdateScript [
   &ParticleState(
    Kill Particles When Lifetime Has Elapsed=false
    Let Infinitely Lived Particles Die When Emitter Deactivates=true
    Loop Particles Lifetime=true
   )
   &SkeletalMeshLocation(
    Bone / Socket Index=$Particles.PartitionParticles.Partition
    MaskSpawnGroup=false
    Mesh Sampling Type=::NewEnumerator5
    Skeletal Mesh=$User.SkeletalMesh
    Socket Sampling Mode=::NewEnumerator5
   )
   &CameraOffset()
   &ScaleSpriteSize(
    Scale Sprite Size Mode=::NewEnumerator0
    Uniform Scale Factor=&MakeCustomFloatFromBool(
     Bool=&SetBoolByIntComparison(A=$Particles.PartitionParticles.Partition, B=1, Comparison Type=::NewEnumerator0)
     True Float=0
     False Float=1
    )
   )
   &ScaleColor(
    Scale Alpha=&Lerp_Float(
     A=0.25
     B=&Multiply_Float(
      A=1
      B=&FloatFromCurve(
       FloatCurve=DataInterface() {propertyValues {Curve {keys [
           <at 0 1 RCIM_Cubic>
           {arriveTangent=0.1177, interpMode=RCIM_Cubic, leaveTangent=0.1177, time=0.0746, value=0.8987}
           {arriveTangent=-0.0227, interpMode=RCIM_Cubic, leaveTangent=-0.0227, tangentMode=RCTM_User, time=0.1682, value=1.0198}
           {arriveTangent=-0.7897, interpMode=RCIM_Cubic, leaveTangent=-0.7897, time=0.2019, value=0.7756}
           {arriveTangent=0.8345, interpMode=RCIM_Cubic, leaveTangent=0.8345, time=0.2748, value=0.9356}
           {arriveTangent=0.5127, interpMode=RCIM_Cubic, leaveTangent=0.5127, time=0.3464, value=0.8961}
           {arriveTangent=-0.6889, interpMode=RCIM_Cubic, leaveTangent=-0.6889, time=0.4074, value=1.0036}
           {arriveTangent=-0.8737, interpMode=RCIM_Cubic, leaveTangent=-0.8737, time=0.4365, value=0.834}
           {arriveTangent=-0.4832, interpMode=RCIM_Cubic, leaveTangent=-0.4832, time=0.4871, value=0.9339}
           {arriveTangent=0.1715, interpMode=RCIM_Cubic, leaveTangent=0.1715, time=0.5331, value=0.7873}
           {arriveTangent=0.795, interpMode=RCIM_Cubic, leaveTangent=0.795, time=0.6112, value=0.9552}
           {arriveTangent=0.9015, interpMode=RCIM_Cubic, leaveTangent=0.9015, time=0.69, value=0.912}
           {arriveTangent=-0.4254, interpMode=RCIM_Cubic, leaveTangent=-0.4254, time=0.7536, value=1.0835}
           {arriveTangent=-1.4701, interpMode=RCIM_Cubic, leaveTangent=-1.4701, time=0.7995, value=0.8654}
           {arriveTangent=-0.4912, interpMode=RCIM_Cubic, leaveTangent=-0.4912, time=0.8674, value=0.9161}
           {arriveTangent=0.6326, interpMode=RCIM_Cubic, leaveTangent=0.6326, time=0.9138, value=0.8093}
           {interpMode=RCIM_Cubic, leaveTangent=0.1355, time=1, value=1}
          ]}}}
       Scale Curve=&FNTimeOfDayScalar(Night Scale=0.4, Morning Scale=1, Evening Scale=0.5)
      )
     )
     Alpha=&ClampFloat(Float=&Add_Float(
       A=$System.Timeline.NormalizedTime
       B=&FloatFromCurve(
        FloatCurve=DataInterface() {propertyValues {Curve {keys [
            {tangentMode=RCTM_User, time=0, value=1}
            {arriveTangent=-0.0405, interpMode=RCIM_Cubic, leaveTangent=-0.0405, tangentMode=RCTM_User, time=1.2, value=1}
            {arriveTangent=-8.0089, interpMode=RCIM_Cubic, leaveTangent=-8.0089, tangentMode=RCTM_User, time=1.5, value=0}
           ]}}}
        CurveIndex=$System.Age
       )
      ))
    )
    Scale RGB=&MakeVectorFromLinearColorRGB(Color=&FNTimeOfDayColor(
      Night Color=<rgba 0.0966 0.0966 0.786 1>
      Morning Color=<rgba 0.0966 0.3264 0.786 1>
      Day Color=<rgba 0.0966 0.3264 0.786 1>
      Evening Color=<rgba 0.0966 0.0966 0.786 1>
     ))
    ScaleA=true
    ScaleRGB=true
   )
   /SetVariables(
    Particles.CoronaColor:LinearColor=&MultiplyLinearColor(
     Linear Color=&FNTimeOfDayColor(
      Night Color=<rgba 0.2995 1.6198 2 1>
      Morning Color=<rgba 0.2995 1.6198 2 1>
      Day Color=<rgba 0.2995 1.6198 2 1>
      Evening Color=<rgba 0.2995 1.6198 2 1>
     )
     Scale Factor=&MakeFloat4FromVectorAndFloat(
      "Vector (RGB)"=<2 2 2>
      "Float (Alpha)"=&FN_FrontEndFloatModifier(
       Gameplay=&Lerp_Float(
        A=0.2
        B=1
        Alpha=&ClampFloat(Float=&Add_Float(A=$System.Timeline.NormalizedTime, B=$System.Size scale))
       )
       Front End=1
      )
     )
    )
    Particles.CoronaPosition:NiagaraPosition=&Add_Vector(
     A=Linked() {linkedVariable {name=Output.SkeletalMeshLocation.SampledPositionWithoutOffset, type=NiagaraPosition}}
     B=&RotateVectorByQuaternion(Vector To Rotate=<0 0 -10>, Quaternion=$Particles.SkeletalMeshLocation.SampledRotation)
    )
    Particles.CoronaSize:Vector2f=<50 50>
   ) {isSetParameters=true}
   &FN_Light_Attributes(
    Color=<rgba 0.298 1 1 3>
    Create Color Attribute=true
    Create Exponent Attribute=true
    Create Position Attribute=true
    Create Radius Attribute=true
    Exponent=2
    Position=&Add_Vector(
     A=$Output.SkeletalMeshLocation.SampledPositionWithoutOffset
     B=&RotateVectorByQuaternion(Vector To Rotate=<0 0 -40>, Quaternion=$Particles.SkeletalMeshLocation.SampledRotation)
    )
    Radius=60
    Scale Color=&Add_Float(A=$System.Size scale, B=$System.Timeline.NormalizedTime)
   )
  ]
  rendererClasses [&NiagaraSpriteRendererProperties, &NiagaraLightRendererProperties]
  renderers [
   {rendererClass=&NiagaraSpriteRendererProperties}
   {rendererClass=&NiagaraLightRendererProperties, rendererIndex=1}
   {rendererClass=&NiagaraSpriteRendererProperties, rendererIndex=2}
  ]
 }
 FireMesh {
  emitterUpdateScript [
   &EmitterState()
   &SpawnRate(SpawnRate=40, Use Spawn Probability=false)
  ]
  particleSpawnScript [
   &PartitionParticles()
   &InitializeParticle(
    Color=<rgba 3 3 3 1>
    Lifetime Max=0.27
    Lifetime Min=0.14
    Lifetime Mode=::NewEnumerator1
    Mesh Scale=&RandomRangeVector(Minimum=<0.12 0.12 0.3>, Maximum=<0.125 0.125 0.3>)
    Mesh Scale Mode=::NewEnumerator4
    UsePositionOffset=false
    Write Material Random=true
   )
   &InitialMeshOrientation(
    Mesh Axis To Orient=<0 0 -1>
    Mesh Orientation Mode=::NewEnumerator2
    Orientation Vector=<0 0 1>
    Rotation=&RandomRangeVector(Minimum=<0 0 -1>, Maximum=<0 0 1>)
    Use Rotation Vector=true
   )
   &CameraOffset() {enabled=false}
  ]
  particleUpdateScript [
   &ParticleState()
   &SkeletalMeshLocation(
    Bone / Socket Index=$Particles.PartitionParticles.Partition
    MaskSpawnGroup=false
    Mesh Sampling Type=::NewEnumerator5
    Sampled Position Offset=<0 1 -5>
    Skeletal Mesh=$User.SkeletalMesh
    Socket Sampling Mode=::NewEnumerator5
   )
   &ScaleMeshSize(Scale Factor=&FN_FrontEndVector3Modifier(Gameplay=<1 1 1>, Front End=&VectorFromFloat(Value=0.95)))
   &ScaleColor(
    Scale Alpha=&FloatFromCurve(FloatCurve=DataInterface() {propertyValues {Curve {keys [
         {arriveTangent=9.0725, interpMode=RCIM_Cubic, leaveTangent=9.0725, tangentMode=RCTM_User, time=0, value=0}
         {arriveTangent=-1.6443, interpMode=RCIM_Cubic, leaveTangent=-1.6443, tangentMode=RCTM_User, time=0.1408, value=0.9847}
         {arriveTangent=-0.0405, interpMode=RCIM_Cubic, leaveTangent=-0.0405, tangentMode=RCTM_User, time=1, value=0}
        ]}}})
    Scale RGB=&VectorFromFloat(Value=&Add_Float(A=$System.Size scale, B=$System.Timeline.NormalizedTime))
    ScaleA=true
    ScaleRGB=true
   )
   &DynamicMaterialParameters(Module.Param0WriteEnabled=true, Module.Param1WriteEnabled=true, Module.Param2WriteEnabled=true, Module.Param3WriteEnabled=true)
  ]
  rendererClasses [&NiagaraMeshRendererProperties]
  renderers [{rendererClass=&NiagaraMeshRendererProperties, rendererIndex=0}]
 }
 SideWings001 {
  emitterSpawnScript [&FN_RibbonIDOffsetUtility()]
  emitterUpdateScript [
   &EmitterState()
   &FN_SpawnAnimTrail(Number of Anim Trails=2)
  ]
  particleSpawnScript [
   &FN_InitializeAnimTrailBoneIndecies(First Bone or Socket Index=2, Second Bone or Socket Index=3)
   &FN_InitializeAnimTrailBoneIndecies(AnimTrail Selection=1, First Bone or Socket Index=4, Second Bone or Socket Index=3) {module=FN_InitializeAnimTrailBoneIndecies001}
   &FN_InitializeAnimTrail(
    AnimTrailWidth=1.1
    Color=&FNTimeOfDayColor(Night Color=<rgba 1 1 1 0.15>, Morning Color=<rgba 1 1 1 0.28>, Day Color=<rgba 1 1 1 0.25>, Evening Color=<rgba 1 1 1 0.28>)
    Lifetime=0.06
   )
   &AddVelocity(Velocity=&MakeCustomVectorFromBool(Bool=$User.bIsFrontEnd, True Vector=<0 -2500 0>, False Vector=<0 2500 0>))
  ]
  particleUpdateScript [
   &SolveForcesAndVelocity(Clamp Velocity=false, Limit Acceleration=false)
   &FN_UpdateAnimTrail(SkeletalMesh=$User.SkeletalMesh)
   /SetVariables(Particles.RibbonU0Override=$Particles.NormalizedAge) {isSetParameters=true}
   &ScaleColor(
    Scale Alpha=&FN_FrontEndFloatModifier(Gameplay=&Lerp_Float(A=0, B=0.8, Alpha=$System.Timeline.NormalizedTime), Front End=1)
    ScaleA=true
    ScaleRGB=true
   )
  ]
  rendererClasses [&NiagaraRibbonRendererProperties, &NiagaraSpriteRendererProperties]
  renderers [
   {rendererClass=&NiagaraRibbonRendererProperties}
   {rendererClass=&NiagaraSpriteRendererProperties, rendererIndex=1}
  ]
 }
}
summary {
 systemName=NS_Glider_IronBlaze_Thrusters
 emitters [
  {emitterName=FireMesh, rendererClasses [&NiagaraMeshRendererProperties]}
  {emitterName=FakeLighting, rendererClasses [&NiagaraSpriteRendererProperties, &NiagaraLightRendererProperties]}
  {emitterName=SideWings001, rendererClasses [&NiagaraRibbonRendererProperties, &NiagaraSpriteRendererProperties]}
  {emitterName=AirThrusters003, rendererClasses [&NiagaraSpriteRendererProperties]}
 ]
 userVariables [
  User.bIsFrontEnd {type=NiagaraBool, defaultValue=true}
  User.bIsFrontEndPreview {type=NiagaraBool, defaultValue=true}
  User.bIsFullyDeployed {type=NiagaraBool, defaultValue=true}
  User.ForwardDot {type=NiagaraFloat, defaultValue=1, description=""}
  User.RibbonIDOffset {type=NiagaraInt32}
  User.RightDot {type=NiagaraFloat, defaultValue=0, description=""}
  User.SkeletalMesh {
   type=NiagaraDataInterfaceSkeletalMesh
   defaultValue {
    struct=NiagaraExt_VariableValue_DataInterface
    value {dataInterface=&self:NiagaraDataInterfaceSkeletalMesh_1}
   }
  }
 ]
}
